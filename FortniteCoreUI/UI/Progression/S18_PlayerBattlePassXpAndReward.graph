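# Season 18 battle pass XP bar widget, built on FortSeasonDataWidget, showing the current level, XP to the next level and a progress track.
# Level and progress events refresh both rich texts and drive the track material's Progress_Start parameter; supercharge sets IsSupercharged.
# SetCollapsedMode plays one animation forward or reversed to narrow the widget and fade out the XP text.
# Construct adjusts offsets and track size for mobile, and the LevelRewards container starts collapsed.
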WidgetBlueprint /FortniteCoreUI/UI/Progression/S18_PlayerBattlePassXpAndReward
animations [Collapse {
  binding_count=2
  length_seconds=0.2
  object=&self:Collapse
  sequence {bindings [
    SizeBox_0 {
     id=#1
     possessed_class=SizeBox
     tracks [{
       type=MovieSceneFloatTrack
       display_name=Width Override
       sections [{type=MovieSceneFloatSection, end_frame=4, start_frame=0, channels [None {type=MovieSceneScriptingFloatChannel, keys [[0,450],[4,170]]}]}]
      }]
    }
    RichText_Level {
     id=#2
     possessed_class=CommonRichTextBlock
     tracks [
      {
       type=MovieSceneFloatTrack
       display_name=Render Opacity
       sections [{type=MovieSceneFloatSection, end_frame=4, start_frame=0, channels [None {type=MovieSceneScriptingFloatChannel, keys [[0,1],[4,0]]}]}]
      }
      {
       type=MovieScene2DTransformTrack
       sections [{
         type=MovieScene2DTransformSection
         end_frame=4
         start_frame=0
         channels [
          Translation.X {type=MovieSceneScriptingFloatChannel, keys [[0,0],[4,50]]}
          Translation.Y {type=MovieSceneScriptingFloatChannel, const [0,1]}
          Angle {type=MovieSceneScriptingFloatChannel, const [0,1]}
          Scale.X {type=MovieSceneScriptingFloatChannel, const [1,1]}
          Scale.Y {type=MovieSceneScriptingFloatChannel, const [1,1]}
          Shear.X {type=MovieSceneScriptingFloatChannel, const [-8,1]}
          Shear.Y {type=MovieSceneScriptingFloatChannel, const [0,1]}
         ]
        }]
      }
     ]
    }
   ]}
 }]
graphs [
 EventGraph {nodes [
   Event Construct()
   > n1 = Branch(Condition=IsMobileGame())
   > SetRenderTranslation(self=Get Image_RaceTrack() -> (Image_RaceTrack:Image Object Reference), Translation=(X=0,Y=-3))
   > n2 = SetRenderTranslation(self=Get RichText_Level() -> (RichText_Level:Common Rich Text Block Object Reference), Translation=(X=0,Y=-10))
   SetDesiredSizeOverride!(
    execute=[@n3.then, @n2.then]
    self:Image Object Reference=Get Image_RaceTrack()
    DesiredSize:Vector 2D Structure=Select(Option 0:Vector 2D Structure=(X=20.7877,Y=15), Option 1:Vector 2D Structure=(X=20.7877,Y=10), Index=IsMobileGame()) -> (ReturnValue:Vector 2D Structure)
   ) -> (then)
   SetRenderTranslation(execute=@n1.else, self=Get Image_RaceTrack(), Translation=(X=0,Y=0))
   > n3 = SetRenderTranslation(self=Get RichText_Level(), Translation=(X=0,Y=0))
   Event OnSuperCharged() {is=title K2Node_Event}
   > n4 = GetDynamicMaterial(self=Get Image_RaceTrack())
   > SetScalarParameterValue(self=@n4, ParameterName=IsSupercharged, Value=1)
   Event OnLevelChanged() {is=title K2Node_Event}
   > Update Text()
   Event OnProgressUpdated() {is=title K2Node_Event}
   > Update Text()
  ]}
 SetCollapsedMode {nodes [
   n1 = SetCollapsedMode!() -> (then, IsCollapsed:bool=~) {is=title K2Node_FunctionEntry}
   > PlayAnimation(
    InAnimation=Get Collapse() -> (Collapse:Widget Animation Object Reference)
    PlayMode=Select(Option 0:EUMGSequencePlayMode Enum=Reverse, Option 1:EUMGSequencePlayMode Enum, Index=@n1) -> (ReturnValue:EUMGSequencePlayMode Enum)
   )
  ]}
 Update Text {nodes [
   Update Text() {is=title K2Node_FunctionEntry}
   n1 = Get XpToNextLevel() -> (XpToNextLevel:int=0)
   n2 = Get CurrentXP() -> (CurrentXP:int=0)
   n3 = Get CurrentLevel() -> (CurrentLevel:int=0)
   > SetText(
    self=Get RichText_Level()
    InText=Format Text!(
     Format="LOCTEXT(\"<text color=\\\"40DFFF\\\">{XPamount} XP</> to LVL <text size=\\\"17\\\">{level}</>\")"
     XPamount:int=int - int(A=@n1, B=@n2) {is=title K2Node_CallFunction}
     level:int=int + int(A=@n3, B=1) -> (ReturnValue=0) {is=title K2Node_CommutativeAssociativeBinaryOperator}
    ) -> (Result)
   )
   > SetText(
    self=Get RichText_CurrentLevel() -> (RichText_CurrentLevel:Common Rich Text Block Object Reference)
    InText=Format Text!(Format="LOCTEXT(\"LVL <text color=\\\"CCF8FF\\\" size=\\\"17\\\">{level}</>\")", level=@n3) -> (Result)
   )
   > n4 = GetDynamicMaterial(self=Get Image_RaceTrack())
   > SetScalarParameterValue(self=@n4, ParameterName=Progress_Start, Value=Get CurrentProgress() -> (CurrentProgress:float))
  ]}
]
variables [Max BPLevel for Rewards {type=integer, default=200}]
widgets {
 info {inheritedWidgetCount=2, parentClass=/Script/FortniteUI.FortSeasonDataWidget, rootWidgetClass=&SizeBox, widgetCount=9}
 widgets [
  SizeBox_0 {
   widgetClassPath=&SizeBox
   widget=SizeBox_0
   properties {bOverride_HeightOverride=true, bOverride_WidthOverride=true, heightOverride=45, widthOverride=450, renderTransformPivot=<1 0.5>}
  }
  VerticalBox_22 {widgetClassPath=&VerticalBox, parent=SizeBox_0, widget=VerticalBox_22, slot {class=SizeBoxSlot}}
  Overlay_0 {
   widgetClassPath=&Overlay
   parent=VerticalBox_22
   widget=Overlay_0
   slot {class=VerticalBoxSlot, properties {padding=<ltrb 0 0 0 8>}}
  }
  RichText_CurrentLevel {
   widgetClassPath=&CommonRichTextBlock
   bIsVariable=true
   parent=Overlay_0
   widget=RichText_CurrentLevel
   properties {
    bOverrideDefaultStyle=true
    lineHeightPercentage=0.8
    text="LVL <text color=\"CCF8FF\" size=\"17\">X</>"
    textTransformPolicy=ToUpper
    defaultTextStyleOverride {
     overflowPolicy=Clip
     transformPolicy=None
     colorAndOpacity {colorUseRule=UseColor_Specified, specifiedColor=<rgba 0.0331 0.402 1 1>}
     font {
      bForceMonospaced=false
      bMaterialIsStencil=false
      fontMaterial=null
      fontObject=&BurbankSmall
      letterSpacing=0
      monospacedWidth=1
      size=14
      skewAmount=0
      typefaceFontName=Black
      outlineSettings {
       bApplyOutlineToDropShadows=false
       bMiteredCorners=false
       bSeparateFillAlpha=false
       outlineMaterial=null
       outlineSize=0
       outlineColor=<rgba 0 0 0 1>
      }
     }
     highlightColor {colorUseRule=UseColor_Specified, specifiedColor=<rgba 0 0 0 1>}
     highlightShape {
      bIsDynamicallyLoaded=false
      drawAs=Image
      imageType=NoImage
      mirroring=NoMirror
      resourceName=None
      resourceObject=null
      tiling=NoTile
      imageSize=<32 32>
      margin=<ltrb 0 0 0 0>
      outlineSettings {
       bUseBrushTransparency=false
       roundingType=HalfHeightRadius
       width=0
       color {colorUseRule=UseColor_Specified, specifiedColor=<rgba 0 0 0 0>}
       cornerRadii=<0 0 0 0>
      }
      tintColor {colorUseRule=UseColor_Specified, specifiedColor=<rgba 1 1 1 1>}
      uVRegion {bIsValid=false, max=<0 0>, min=<0 0>}
     }
     selectedBackgroundColor {colorUseRule=UseColor_Foreground, specifiedColor=<rgba 1 0 1 1>}
     shadowColorAndOpacity=<rgba 0 0 0 1>
     shadowOffset=<0 0>
     strikeBrush {
      bIsDynamicallyLoaded=false
      drawAs=Image
      imageType=NoImage
      mirroring=NoMirror
      resourceName=None
      resourceObject=null
      tiling=NoTile
      imageSize=<32 32>
      margin=<ltrb 0 0 0 0>
      outlineSettings {
       bUseBrushTransparency=false
       roundingType=HalfHeightRadius
       width=0
       color {colorUseRule=UseColor_Specified, specifiedColor=<rgba 0 0 0 0>}
       cornerRadii=<0 0 0 0>
      }
      tintColor {colorUseRule=UseColor_Specified, specifiedColor=<rgba 1 1 1 1>}
      uVRegion {bIsValid=false, max=<0 0>, min=<0 0>}
     }
     underlineBrush {
      bIsDynamicallyLoaded=false
      drawAs=Image
      imageType=NoImage
      mirroring=NoMirror
      resourceName=None
      resourceObject=null
      tiling=NoTile
      imageSize=<32 32>
      margin=<ltrb 0 0 0 0>
      outlineSettings {
       bUseBrushTransparency=false
       roundingType=HalfHeightRadius
       width=0
       color {colorUseRule=UseColor_Specified, specifiedColor=<rgba 0 0 0 0>}
       cornerRadii=<0 0 0 0>
      }
      tintColor {colorUseRule=UseColor_Specified, specifiedColor=<rgba 1 1 1 1>}
      uVRegion {bIsValid=false, max=<0 0>, min=<0 0>}
     }
    }
    renderTransform {angle=0, scale=<1 1>, shear=<-8 0>, translation=<0 0>}
   }
   slot {class=OverlaySlot, properties {verticalAlignment=VAlign_Bottom}}
  }
  RichText_Level {
   widgetClassPath=&CommonRichTextBlock
   bIsVariable=true
   parent=Overlay_0
   widget=RichText_Level
   properties {
    bOverrideDefaultStyle=true
    lineHeightPercentage=0.8
    mobileTextBlockScale=1.3
    text="<text color=\"40DFFF\">0 XP</> to LVL <text size=\"17\">0</>"
    defaultTextStyleOverride {
     overflowPolicy=Clip
     transformPolicy=None
     colorAndOpacity {colorUseRule=UseColor_Specified, specifiedColor=<rgba 0.0331 0.402 1 1>}
     font {
      bForceMonospaced=false
      bMaterialIsStencil=false
      fontMaterial=null
      fontObject=&BurbankSmall
      letterSpacing=0
      monospacedWidth=1
      size=14
      skewAmount=0
      typefaceFontName=Black
      outlineSettings {
       bApplyOutlineToDropShadows=false
       bMiteredCorners=false
       bSeparateFillAlpha=false
       outlineMaterial=null
       outlineSize=0
       outlineColor=<rgba 0 0 0 1>
      }
     }
     highlightColor {colorUseRule=UseColor_Specified, specifiedColor=<rgba 0 0 0 1>}
     highlightShape {
      bIsDynamicallyLoaded=false
      drawAs=Image
      imageType=NoImage
      mirroring=NoMirror
      resourceName=None
      resourceObject=null
      tiling=NoTile
      imageSize=<32 32>
      margin=<ltrb 0 0 0 0>
      outlineSettings {
       bUseBrushTransparency=false
       roundingType=HalfHeightRadius
       width=0
       color {colorUseRule=UseColor_Specified, specifiedColor=<rgba 0 0 0 0>}
       cornerRadii=<0 0 0 0>
      }
      tintColor {colorUseRule=UseColor_Specified, specifiedColor=<rgba 1 1 1 1>}
      uVRegion {bIsValid=false, max=<0 0>, min=<0 0>}
     }
     selectedBackgroundColor {colorUseRule=UseColor_Foreground, specifiedColor=<rgba 1 0 1 1>}
     shadowColorAndOpacity=<rgba 0 0 0 1>
     shadowOffset=<0 0>
     strikeBrush {
      bIsDynamicallyLoaded=false
      drawAs=Image
      imageType=NoImage
      mirroring=NoMirror
      resourceName=None
      resourceObject=null
      tiling=NoTile
      imageSize=<32 32>
      margin=<ltrb 0 0 0 0>
      outlineSettings {
       bUseBrushTransparency=false
       roundingType=HalfHeightRadius
       width=0
       color {colorUseRule=UseColor_Specified, specifiedColor=<rgba 0 0 0 0>}
       cornerRadii=<0 0 0 0>
      }
      tintColor {colorUseRule=UseColor_Specified, specifiedColor=<rgba 1 1 1 1>}
      uVRegion {bIsValid=false, max=<0 0>, min=<0 0>}
     }
     underlineBrush {
      bIsDynamicallyLoaded=false
      drawAs=Image
      imageType=NoImage
      mirroring=NoMirror
      resourceName=None
      resourceObject=null
      tiling=NoTile
      imageSize=<32 32>
      margin=<ltrb 0 0 0 0>
      outlineSettings {
       bUseBrushTransparency=false
       roundingType=HalfHeightRadius
       width=0
       color {colorUseRule=UseColor_Specified, specifiedColor=<rgba 0 0 0 0>}
       cornerRadii=<0 0 0 0>
      }
      tintColor {colorUseRule=UseColor_Specified, specifiedColor=<rgba 1 1 1 1>}
      uVRegion {bIsValid=false, max=<0 0>, min=<0 0>}
     }
    }
    renderTransform {angle=0, scale=<1 1>, shear=<-8 0>, translation=<0 0>}
   }
   slot {class=OverlaySlot, properties {horizontalAlignment=HAlign_Right, verticalAlignment=VAlign_Bottom}}
  }
  Image_RaceTrack {
   widgetClassPath=&Image
   bIsVariable=true
   parent=VerticalBox_22
   widget=Image_RaceTrack
   properties {
    visibility=SelfHitTestInvisible
    brush {
     bIsDynamicallyLoaded=false
     drawAs=Image
     imageType=NoImage
     mirroring=NoMirror
     resourceName=None
     resourceObject=/FortUILibrary/Foundation/Art_ToolKit/ProgressBars/M_UI_XP_Everywhere_ProgressBar_QuestScreen
     tiling=NoTile
     imageSize=<20.7877 10>
     margin=<ltrb 0.5 0 0.5 0>
     outlineSettings {
      bUseBrushTransparency=false
      roundingType=HalfHeightRadius
      width=0
      color {colorUseRule=UseColor_Specified, specifiedColor=<rgba 0 0 0 0>}
      cornerRadii=<0 0 0 0>
     }
     tintColor {colorUseRule=UseColor_Specified, specifiedColor=<rgba 1 1 1 1>}
     uVRegion {bIsValid=false, max=<0 0>, min=<0 0>}
    }
    renderTransform {angle=0, scale=<1 1>, shear=<-8 0>, translation=<0 0>}
   }
   slot {class=VerticalBoxSlot}
  }
  ScaleBox_172 {
   widgetClassPath=&ScaleBox
   parent=VerticalBox_22
   widget=ScaleBox_172
   properties {visibility=Collapsed}
   slot {class=VerticalBoxSlot}
  }
  LevelRewards {
   widgetClassPath=/FortniteCoreUI/HUD/MiniMap/AthenaChallengeRewards
   bInherited=true
   bIsVariable=true
   parent=ScaleBox_172
   widget=LevelRewards
   slot {class=ScaleBoxSlot, properties {horizontalAlignment=HAlign_Fill, verticalAlignment=VAlign_Fill}}
  }
  SocialAvatarIcon {widgetClassPath=/Script/FortniteUI.FortSocialAvatarIcon, bInherited=true, widget=null}
 ]
}
